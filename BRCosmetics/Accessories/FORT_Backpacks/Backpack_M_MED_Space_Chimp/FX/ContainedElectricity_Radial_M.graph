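# Additive unlit, two-sided Niagara sprite material that draws radial contained electricity for the Space Chimp backpack.
# A lightning texture is warped by stepped panning noise, then scaled, rotated and offset using dynamic parameters and quantized sine-based time for jitter.
# Panning smoke erodes the lightning, and an inverted sphere mask adds rim glow.
# Particle color tints the result, which is circle-masked, depth-faded and passed through shared emissive adjustments.
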Material /BRCosmetics/Accessories/FORT_Backpacks/Backpack_M_MED_Space_Chimp/FX/ContainedElectricity_Radial_M
graph {
 properties {
  bUsedWithNiagaraSprites=true
  blendMode=BLEND_Additive
  pixelDepthOffsetMode=PDOM_Legacy
  shadingModel=MSM_Unlit
  twoSided=true
  usedShadingModels=Unlit
 }
 nodes [
  n1 = Ceil(None=Multiply(A=Time()) {constB=16})
  n2 = TextureSample(UVs=Panner(Coordinate=TextureCoordinate() {uTiling=0.4, vTiling=0.4}, Time=Divide(A=@n1) {constB=4}) {speedX=0.4, speedY=0.4}) {samplerType=SAMPLERTYPE_LinearColor, texture=/Game/Global/Textures/Default/Noise/TilingNoise06_N_uncompressed}
  n3 = Multiply(A=Subtract(A=ComponentMask(None=@n2.RGB) {g=true, r=true}) {constB=0.5}) {constB=0.1}
  n4 = TextureCoordinate()
  n5 = DynamicParameter()
  n6 = Time()
  n7 = Add(A=Sine(None=@n6) {period=1.8425}, B=Sine(None=@n6) {period=3.7854})
  n8 = Divide(A=Ceil(None=Multiply(A=@n7) {constB=2.1})) {constB=4}
  n9 = Rotator(
   Coordinate=MaterialFunctionCall(UVs=Add(A=@n3, B=@n4) {constB=0}, Texture Scale=LinearInterpolate(Alpha=@n5.Param3) {constA=1.2, constB=1.6}) {materialFunction=&ScaleUVsByCenter}.UVs
   Time=Add(A=Multiply(A=@n5.Param2) {constB=10}, B=@n8)
  ) {speed=0.5}
  n10 = Add(A=@n5.Param1, B=@n8)
  n11 = TextureSample(UVs=Panner() {speedX=0.4, speedY=0.4}) {texture=/Game/Global/Textures/Default/Tileable_Patterns/T_Blowing_Smoke_Pattern2_Tile_DISP}
  n12 = Constant() {r=2}
  n13 = Subtract(
   A=TextureSample(UVs=Add(A=@n9, B=@n10)) {texture=/Game/Weapons/WeaponSkins/Textures/T_Lightning_Drift_v2}.G
   B=LinearInterpolate(Alpha=Power(Base=@n11.R, Exp=@n12))
  )
  n14 = SphereMask(A=TextureCoordinate(), B=Constant2Vector() {g=0.5, r=0.5}) {attenuationRadius=0.5, hardnessPercent=0}
  n15 = Power(Base=OneMinus(None=@n14), Exp=Constant() {r=5})
  n16 = Constant() {r=1}
  n17 = Multiply(A=ParticleColor().RGB, B=Add(A=Saturate(None=@n13), B=Multiply(A=@n15, B=@n16)))
  n18 = Power(Base=@n14)
  MaterialFunctionCall(Emissive=DepthFade(Opacity=Multiply(A=@n17, B=@n18)) {fadeDistanceDefault=16}) {materialFunction=&MF_FX_EmissiveAdjustments}
 ]
}
