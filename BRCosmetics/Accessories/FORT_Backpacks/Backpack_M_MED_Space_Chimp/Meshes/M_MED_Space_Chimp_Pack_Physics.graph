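# Physics asset for the Space Chimp backpack cosmetic, providing bodies for its dangling jaw, two arms and three bananas.
# The pack body is a cluster of eight large overlapping capsules, while each dyn_ bone has a single small capsule.
# Every dyn_ bone is constrained directly to the pack, with no chaining.
# Jaw and arms lock twist and one swing axis to act as hinges, while bananas lock no angular axis.
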PhysicsAsset /BRCosmetics/Accessories/FORT_Backpacks/Backpack_M_MED_Space_Chimp/Meshes/M_MED_Space_Chimp_Pack_Physics
bodies {
 dyn_arm_l [{shapeType=Capsule, length=2, radius=2, shapeName=dyn_arm_l_capsule, center=<1.2146 -0.0001 0>}]
 dyn_arm_r [{shapeType=Capsule, length=2, radius=2, shapeName=dyn_arm_r_capsule, center=<0.9393 0 0>}]
 dyn_banana_1 [{
   shapeType=Capsule
   length=13.3153
   radius=2.234
   shapeName=dyn_banana_1_capsule
   center=<0.279 -0.2602 -1.1245>
   rotation=<rot -44.9958 1.9256 24.0023>
  }]
 dyn_banana_2 [{
   shapeType=Capsule
   length=8.8227
   radius=2
   shapeName=dyn_banana_2_capsule
   center=<-0.6411 0.2446 -0.0716>
   rotation=<rot 0.6072 23.6636 -91.4176>
  }]
 dyn_banana_3 [{
   shapeType=Capsule
   length=9.8064
   radius=2
   shapeName=dyn_banana_3_capsule
   center=<-0.4174 -0.7233 -1.1404>
   rotation=<rot 7.8574 42.7208 -105.457>
  }]
 dyn_jaw [{shapeType=Capsule, length=2.2637, radius=2.2994, shapeName=dyn_jaw_capsule, center=<2.3239 0 0>, rotation=<rot 0 0 90>}]
 pack [
  {shapeType=Capsule, length=22.6404, radius=25.6469, center=<9.5992 21.0108 -9.8712>, rotation=<rot 14.318 -26.2456 -6.9537>}
  {
   shapeType=Capsule
   length=22.6404
   radius=25.6469
   shapeName=pack_capsule
   center=<-23.1334 6.7029 -18.3762>
   rotation=<rot 13.1043 -45.3622 -11.5775>
  }
  {
   shapeType=Capsule
   length=22.6404
   radius=25.6469
   shapeName=pack_capsule
   center=<-22.5365 -28.057 -17.8247>
   rotation=<rot 13.2086 -44.8394 -11.4624>
  }
  {
   shapeType=Capsule
   length=22.6404
   radius=25.6469
   shapeName=pack_capsule
   center=<29.1484 -27.9494 -1.7129>
   rotation=<rot 13.1043 -45.3684 -11.5829>
  }
  {
   shapeType=Capsule
   length=22.6404
   radius=25.6469
   shapeName=pack_capsule
   center=<28.7813 4.4176 -2.5976>
   rotation=<rot 6.3031 -72.6424 -16.2818>
  }
  {
   shapeType=Capsule
   length=22.6404
   radius=25.6469
   shapeName=pack_capsule
   center=<6.1846 -41.8591 -7.8413>
   rotation=<rot 6.2227 -72.9164 -16.3156>
  }
  {
   shapeType=Capsule
   length=14.1563
   radius=21.7662
   shapeName=pack_capsule
   center=<13.5534 -9.5405 -38.8775>
   rotation=<rot -69.7059 -32.1288 30.4982>
  }
  {
   shapeType=Capsule
   length=15
   radius=25.8297
   shapeName=pack_capsule
   center=<-9.331 -11.4417 23.3496>
   rotation=<rot -69.2174 -32.1432 30.4316>
  }
 ]
}
constraints [
 {
  bone1Name=dyn_jaw
  bone2Name=pack
  swing1LimitDegrees=45
  swing1Motion=Locked
  swing2LimitDegrees=30
  twistLimitDegrees=45
  twistMotion=Locked
 }
 {bone1Name=dyn_banana_1, bone2Name=pack, swing1LimitDegrees=45, swing2LimitDegrees=45, twistLimitDegrees=45}
 {bone1Name=dyn_banana_2, bone2Name=pack, swing1LimitDegrees=45, swing2LimitDegrees=45, twistLimitDegrees=45}
 {bone1Name=dyn_banana_3, bone2Name=pack, swing1LimitDegrees=45, swing2LimitDegrees=45, twistLimitDegrees=45}
 {
  bone1Name=dyn_arm_l
  bone2Name=pack
  swing1LimitDegrees=45
  swing2LimitDegrees=45
  swing2Motion=Locked
  twistLimitDegrees=45
  twistMotion=Locked
 }
 {
  bone1Name=dyn_arm_r
  bone2Name=pack
  swing1LimitDegrees=45
  swing2LimitDegrees=45
  swing2Motion=Locked
  twistLimitDegrees=45
  twistMotion=Locked
 }
]
