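# Animation Blueprint for the Space Chimp backpack cosmetic that adds secondary physics motion and animates three dynamic banana bones.
# A RigidBody node simulates from the reference pose using an override physics asset, gravity override and scaled-down component motion.
# Each update converts a sine of real time into three scaled, offset vectors.
# Chained additive Transform Bone nodes apply these as translations after physics, making the bananas oscillate on a shared phase.
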AnimBlueprint /BRCosmetics/Accessories/FORT_Backpacks/Backpack_M_MED_Space_Chimp/Meshes/M_MED_Space_Chimp_Pack_AnimBP
target_skeleton=&/M_MED_Space_Chimp_Pack_Skeleton
graphs [
 AnimGraph {nodes [
   n1 = RigidBody(ComponentPose=Local To Component(LocalPose=Local Space Ref Pose())) {
    bOverrideWorldGravity=true
    overridePhysicsAsset=&/M_MED_Space_Chimp_Pack_Physics
    alphaBoolBlend {blendOption=Linear}
    componentAppliedLinearAccClamp=<1500 1500 1500>
    componentLinearAccScale=<0.05 0.05 0.05>
    componentLinearVelScale=<0.05 0.05 0.08>
    overrideWorldGravity {z=-980}
   }
   n2 = "Transform (Modify) Bone - Bone: dyn_banana_2"(
    ComponentPose="Transform (Modify) Bone - Bone: dyn_banana_1"(ComponentPose=@n1, Alpha=1, Translation=Get banana1XPos() -> (banana1XPos:Vector), Rotation="0,0,0", Scale="1,1,1") {translationMode=BMM_Additive, alphaBoolBlend {blendOption=Linear}, boneToModify {boneName=dyn_banana_1}}
    Alpha=1
    Translation=Get banana2XPos() -> (banana2XPos:Vector)
    Rotation="0,0,0"
    Scale="1,1,1"
   ) {translationMode=BMM_Additive, alphaBoolBlend {blendOption=Linear}, boneToModify {boneName=dyn_banana_2}}
   n3 = Get banana3XPos() -> (banana3XPos:Vector)
   Output Pose(Result=Component To Local(ComponentPose="Transform (Modify) Bone - Bone: dyn_banana_3"(ComponentPose=@n2, Alpha=1, Translation=@n3, Rotation="0,0,0", Scale="1,1,1") {translationMode=BMM_Additive, alphaBoolBlend {blendOption=Linear}, boneToModify {boneName=dyn_banana_3}})) {layerGroup=None}
  ]}
 EventGraph {nodes [
   Event BlueprintUpdateAnimation()
   > n1 = Sequence!(execute) -> (then_0, then_1, then_2:Exec)
   n2 = GetRealTimeSeconds()
   n3 = BreakVector(InVec="To Vector (Float)"!(InDouble:double="Sin (Radians)"(A=@n2)) -> (ReturnValue:Vector))
   n4 = float * float(A=@n3, B=0.2)
   > Set banana1XPos(banana1XPos:Vector=MakeVector(
     X=float + float(A=@n4, B=2) {is=title K2Node_CommutativeAssociativeBinaryOperator}
     Y=float * float(A=@n3.Y, B=0.2)
     Z=float * float(A=@n3.Z, B=0.2)
    )) -> (Output_Get:Vector)
   n5 = float * float(A=@n3.Y, B=-0.6)
   n6 = float * float(A=@n3.Z, B=0.8)
   n7 = float * float(A=@n3, B=-0.6)
   Set banana2XPos(execute=@n1.then_1, banana2XPos:Vector=MakeVector(X=float + float(A=@n7, B=1), Y=float + float(A=@n5, B=2), Z=float + float(A=@n6, B=-1))) -> (Output_Get:Vector)
   Set banana3XPos(
    execute=@n1.then_2
    banana3XPos:Vector=MakeVector(X=float * float(A=@n3, B=0.7), Y=float * float(A=@n3.Y, B=0.9), Z=float * float(A=@n3.Z, B=-0.7))
   ) -> (Output_Get:Vector)
   TryGetPawnOwner()
  ]}
]
variables [
 banana1XPos {type=Vector}
 banana2XPos {type=Vector}
 banana3XPos {type=Vector}
 NewVar_0 {type=Vector}
]
